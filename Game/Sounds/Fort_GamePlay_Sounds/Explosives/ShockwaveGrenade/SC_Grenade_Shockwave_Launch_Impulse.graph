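# Sound cue for the shockwave grenade's impulse launch whoosh, spatialized with its own attenuation and a small stop-oldest concurrency limit.
# IsEnemy and IsLocal branches pick a mix level: loudest for enemies, slightly lower for the local player, much quieter for other non-enemies.
# Each path randomly picks one of the same three whoosh waves through a modulator.
# A smoothed, intensified Doppler node at the root shifts pitch with relative motion.
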SoundCue /Game/Sounds/Fort_GamePlay_Sounds/Explosives/ShockwaveGrenade/SC_Grenade_Shockwave_Launch_Impulse
root=Doppler_0
nodes [
 {id=Branch_0, children [Mixer_1,Branch_1,Branch_1], properties {boolParameterName=IsEnemy}}
 {id=Branch_1, children [Mixer_2,Mixer_0,Mixer_2], properties {boolParameterName=IsLocal}}
 {id=Doppler_0, children [Branch_0], properties {bUseSmoothing=true, dopplerIntensity=2}}
 {id=Mixer_0, children [Modulator_3], properties {inputVolume [0.3]}}
 {id=Mixer_1, children [Modulator_2], properties {inputVolume [1.1]}}
 {id=Mixer_2, children [Modulator_0], properties {inputVolume [0.85]}}
 {id=Modulator_0, children [Random_0]}
 {id=Modulator_2, children [Random_1]}
 {id=Modulator_3, children [Random_3]}
 {id=Random_0, children [WavePlayer_0,WavePlayer_1,WavePlayer_2], properties {weights [1,1,1]}}
 {id=Random_1, children [WavePlayer_6,WavePlayer_7,WavePlayer_8], properties {weights [1,1,1]}}
 {id=Random_3, children [WavePlayer_9, WavePlayer_10, WavePlayer_11], properties {weights [1,1,1]}}
 {
  id=WavePlayer_0
  properties {soundWaveAssetPtr=&/Grenade_Shockwave_Impulse_Launch_Whoosh_01}
 }
 {
  id=WavePlayer_1
  properties {soundWaveAssetPtr=&/Grenade_Shockwave_Impulse_Launch_Whoosh_02}
 }
 {
  id=WavePlayer_10
  properties {soundWaveAssetPtr=&Grenade_Shockwave_Impulse_Launch_Whoosh_02}
 }
 {
  id=WavePlayer_11
  properties {soundWaveAssetPtr=&/Grenade_Shockwave_Impulse_Launch_Whoosh_03}
 }
 {
  id=WavePlayer_2
  properties {soundWaveAssetPtr=&Grenade_Shockwave_Impulse_Launch_Whoosh_03}
 }
 {
  id=WavePlayer_6
  properties {soundWaveAssetPtr=&Grenade_Shockwave_Impulse_Launch_Whoosh_01}
 }
 {
  id=WavePlayer_7
  properties {soundWaveAssetPtr=&Grenade_Shockwave_Impulse_Launch_Whoosh_02}
 }
 {
  id=WavePlayer_8
  properties {soundWaveAssetPtr=&Grenade_Shockwave_Impulse_Launch_Whoosh_03}
 }
 {
  id=WavePlayer_9
  properties {soundWaveAssetPtr=&Grenade_Shockwave_Impulse_Launch_Whoosh_01}
 }
]
properties {
 attenuationSettings=/Game/Sounds/Attenuation/Fort_Grenade_Shockwave_Impulse_Launch_Att
 bOverrideConcurrency=true
 maxDistance=11000
 soundClassObject=&Fort_Gadgets_Items
 volumeMultiplier=0.8
 attenuationOverrides {distanceAlgorithm=NaturalSound, falloffDistance=3800, lPFRadiusMax=3000, lPFRadiusMin=2000, attenuationShapeExtents {x=2200}}
 concurrencyOverrides {maxCount=2, resolutionRule=StopOldest}
}
