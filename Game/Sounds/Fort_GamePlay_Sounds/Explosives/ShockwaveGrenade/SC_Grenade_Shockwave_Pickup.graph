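# Sound cue for picking up the Shockwave Grenade, using weapon pickup attenuation and the gadget items sound class.
# A mixer layers a modulated shockwave pickup wave with a reused Impulse Grenade drop sound.
# A quality-level node plays a random pick of three drop variants on the first two levels and a single drop wave on the other two.
# The cue plays quieter and higher-pitched, and its concurrency override prevents new instances at the limit.
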SoundCue /Game/Sounds/Fort_GamePlay_Sounds/Explosives/ShockwaveGrenade/SC_Grenade_Shockwave_Pickup
root=Mixer_0
nodes [
 {id=Mixer_0, children [Modulator_0, QualityLevel_0], properties {inputVolume [0.8,1]}}
 {id=Modulator_0, children [WavePlayer_1]}
 {id=QualityLevel_0, children [Random_0,Random_0,WavePlayer_4,WavePlayer_4]}
 {id=Random_0, children [WavePlayer_0,WavePlayer_2,WavePlayer_3], properties {weights [1,1,1]}}
 {id=WavePlayer_0, properties {soundWaveAssetPtr=&/ImpulseGrenade_Drop_01}}
 {id=WavePlayer_1, properties {soundWaveAssetPtr=&/Grenade_Shockwave_Pickup01}}
 {id=WavePlayer_2, properties {soundWaveAssetPtr=&/ImpulseGrenade_Drop_02}}
 {id=WavePlayer_3, properties {soundWaveAssetPtr=&/ImpulseGrenade_Drop_03}}
 {id=WavePlayer_4, properties {soundWaveAssetPtr=&ImpulseGrenade_Drop_01}}
]
properties {
 attenuationSettings=&Fort_Weapon_Pickup_Att
 bOverrideConcurrency=true
 maxDistance=3500
 pitchMultiplier=1.2
 soundClassObject=&Fort_Gadgets_Items
 volumeMultiplier=0.4
 concurrencyOverrides {resolutionRule=PreventNew}
}
