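# Sound cue for the Shockwave Grenade drop sound, built from Impulse Grenade drop waves and classed as loot collect audio.
# A Quality Level node feeds its first two slots from an equal-weight random pick of three waves and its last two from a single Drop_01 wave.
# A root Modulator randomizes pitch slightly upward, on top of raised cue pitch and volume, with shared weapon-drop attenuation.
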SoundCue /Game/Sounds/Fort_GamePlay_Sounds/Explosives/ShockwaveGrenade/SC_ShockwaveGrenade_Drop
root=Modulator_0
nodes [
 {id=Modulator_0, children [QualityLevel_0], properties {pitchMax=1.2, pitchMin=1.1}}
 {id=QualityLevel_0, children [Random_0,Random_0,WavePlayer_6,WavePlayer_6]}
 {id=Random_0, children [WavePlayer_3,WavePlayer_4,WavePlayer_5], properties {weights [1,1,1]}}
 {id=WavePlayer_3, properties {soundWaveAssetPtr=&/ImpulseGrenade_Drop_01}}
 {id=WavePlayer_4, properties {soundWaveAssetPtr=&/ImpulseGrenade_Drop_02}}
 {id=WavePlayer_5, properties {soundWaveAssetPtr=&/ImpulseGrenade_Drop_03}}
 {id=WavePlayer_6, properties {soundWaveAssetPtr=&ImpulseGrenade_Drop_01}}
]
properties {
 attenuationSettings=&Fort_Weapon_Drop_Att
 maxDistance=3000
 pitchMultiplier=1.1
 soundClassObject=&Fort_Loot_Collect
 volumeMultiplier=1.2
 attenuationOverrides {falloffDistance=2400, stereoSpread=2000, attenuationShapeExtents {x=800}}
}
