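# Sound cue for the Shockwave Grenade explosion in the weapons sound class, layering close and distant sounds under separate attenuation settings.
# The close branch mixes a pitch-varied blast with a quieter, delayed "Wub" layer, while the distant branch plays at lowered, randomized pitch.
# Each layer's quality node uses either a random pick of three waves or a single fixed wave, depending on quality level.
# Concurrency allows two instances, stopping the oldest.
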SoundCue /Game/Sounds/Fort_GamePlay_Sounds/Explosives/ShockwaveGrenade/SC_ShockwaveGrenade_Explode
root=Mixer_2
nodes [
 {
  id=Attenuation_0
  children [Modulator_13]
  properties {attenuationSettings=/Game/Sounds/Attenuation/Fort_Explosion_Medium_Athena_Grenade_Att}
 }
 {id=Attenuation_1, children [Mixer_0], properties {attenuationSettings=/Game/Sounds/Attenuation/Fort_Explosion_Close_Athena_Grenade_Att}}
 {id=Delay_0, children [Modulator_2], properties {delayMax=0.4, delayMin=0.4}}
 {id=Mixer_0, children [Delay_0,Modulator_0], properties {inputVolume [0.3,1]}}
 {id=Mixer_2, children [Attenuation_1, Attenuation_0], properties {inputVolume [1.2,1.1]}}
 {id=Modulator_0, children [QualityLevel_1], properties {pitchMax=1.1, pitchMin=0.9}}
 {id=Modulator_13, children [QualityLevel_2], properties {pitchMax=0.95, pitchMin=0.85, volumeMax=1.1, volumeMin=1}}
 {id=Modulator_2, children [QualityLevel_0]}
 {id=QualityLevel_0, children [Random_2,Random_2,WavePlayer_3,WavePlayer_3]}
 {id=QualityLevel_1, children [Random_1,Random_1,WavePlayer_6,WavePlayer_6]}
 {id=QualityLevel_2, children [Random_3,Random_3,WavePlayer_9,WavePlayer_9]}
 {id=Random_1, children [WavePlayer_6,WavePlayer_7,WavePlayer_8], properties {weights [1,1,1]}}
 {id=Random_2, children [WavePlayer_3,WavePlayer_4,WavePlayer_5], properties {weights [1,1,1]}}
 {id=Random_3, children [WavePlayer_9, WavePlayer_10, WavePlayer_11], properties {weights [1,1,1]}}
 {
  id=WavePlayer_10
  properties {soundWaveAssetPtr=&/ShockwaveGrenade_Explode_Distant_02}
 }
 {
  id=WavePlayer_11
  properties {soundWaveAssetPtr=&/ShockwaveGrenade_Explode_Distant_03}
 }
 {
  id=WavePlayer_3
  properties {soundWaveAssetPtr=&/ShockwaveGrenade_Explode_Close_Wub_01}
 }
 {
  id=WavePlayer_4
  properties {soundWaveAssetPtr=&/ShockwaveGrenade_Explode_Close_Wub_02}
 }
 {
  id=WavePlayer_5
  properties {soundWaveAssetPtr=&/ShockwaveGrenade_Explode_Close_Wub_03}
 }
 {
  id=WavePlayer_6
  properties {soundWaveAssetPtr=&/ShockwaveGrenade_Explode_Close_01}
 }
 {
  id=WavePlayer_7
  properties {soundWaveAssetPtr=&/ShockwaveGrenade_Explode_Close_02}
 }
 {
  id=WavePlayer_8
  properties {soundWaveAssetPtr=&/ShockwaveGrenade_Explode_Close_03}
 }
 {
  id=WavePlayer_9
  properties {soundWaveAssetPtr=&/ShockwaveGrenade_Explode_Distant_01}
 }
]
properties {
 bOverrideConcurrency=true
 maxDistance=9625
 soundClassObject=/Game/Sounds/FortSoundClassesAndModes/Fort_Weapons
 volumeMultiplier=1
 attenuationOverrides {distanceAlgorithm=NaturalSound, falloffDistance=3800, lPFRadiusMax=3000, lPFRadiusMin=2000, attenuationShapeExtents {x=2200}}
 concurrencyOverrides {maxCount=2, resolutionRule=StopOldest}
}
