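# Sound cue for the shockwave grenade's in-flight audio, routed to the explosives sound class with its own concurrency override.
# A mixer layers a pitched-down looping beep, a looping launcher in-air sound with randomized modulation and exaggerated Doppler, and a one-shot impulse grenade launch sound.
# The launch layer has its own throw attenuation node, nested inside the projectile loop attenuation that wraps the whole mix.
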SoundCue /Game/Sounds/Fort_GamePlay_Sounds/Explosives/ShockwaveGrenade/SC_ShockwaveGrenade_InAirLoop
root=Attenuation_2
nodes [
 {
  id=Attenuation_0
  children [WavePlayer_4]
  properties {attenuationSettings=/Game/Sounds/Attenuation/Fort_Weapon_Projectile_InAir_Launch_Throw_Att}
 }
 {id=Attenuation_2, children [Mixer_0], properties {attenuationSettings=/Game/Sounds/Attenuation/Fort_Weapon_Projectile_InAir_Loop_Att}}
 {id=Doppler_0, children [Modulator_0], properties {dopplerIntensity=10}}
 {id=Mixer_0, children [Modulator_2, Doppler_0, Attenuation_0], properties {inputVolume [1,0.4,1.2]}}
 {id=Modulator_0, children [WavePlayer_1]}
 {id=Modulator_2, children [WavePlayer_3], properties {pitchMax=0.8, pitchMin=0.8}}
 {id=WavePlayer_1, properties {bLooping=true, soundWaveAssetPtr=/Game/Sounds/Fort_Weapons/Launchers/Shockwave/Shockwave_Launcher_InAirLoop_01}}
 {
  id=WavePlayer_3
  properties {bLooping=true, soundWaveAssetPtr=&/Grenade_Shockwave_Beep01}
 }
 {
  id=WavePlayer_4
  properties {soundWaveAssetPtr=&/ImpulseGrenade_Prj_InAir_Launch_01}
 }
]
properties {
 bOverrideConcurrency=true
 maxDistance=8000
 soundClassObject=/Game/Sounds/FortSoundClassesAndModes/Fort_Explosives
 volumeMultiplier=0.3
 attenuationOverrides {distanceAlgorithm=Logarithmic, falloffDistance=1200, spatializationAlgorithm=SPATIALIZATION_HRTF, attenuationShapeExtents {x=500}}
 concurrencyOverrides {resolutionRule=PreventNew}
}
