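# Sound cue for the shockwave grenade landing, using a weapon-drop attenuation and the loot-collect sound class.
# A mixer layers a boosted random Impulse Grenade drop layer with a quieter random Sticky Grenade Launcher drop-land sound.
# The sticky layer starts after a short fixed delay and is pitched up to roughly double.
# Quality-level nodes give the two lower-indexed levels random variations and the other two a single wave per layer.
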SoundCue /Game/Sounds/Fort_GamePlay_Sounds/Explosives/ShockwaveGrenade/SC_ShockwaveGrenade_Land
root=Mixer_0
nodes [
 {id=Delay_0, children [Modulator_0], properties {delayMax=0.05, delayMin=0.05}}
 {id=Mixer_0, children [QualityLevel_1, Delay_0], properties {inputVolume [2,0.8]}}
 {id=Modulator_0, children [QualityLevel_2], properties {pitchMax=2, pitchMin=1.9}}
 {id=QualityLevel_1, children [Random_2,Random_2,WavePlayer_7,WavePlayer_7]}
 {id=QualityLevel_2, children [Random_1,Random_1,WavePlayer_0,WavePlayer_0]}
 {id=Random_1, children [WavePlayer_0,WavePlayer_1,WavePlayer_2], properties {weights [1,1,1]}}
 {id=Random_2, children [WavePlayer_7,WavePlayer_8], properties {weights [1,1]}}
 {
  id=WavePlayer_0
  properties {soundWaveAssetPtr=&/StickyGrenadeLauncher_DropLand_01}
 }
 {
  id=WavePlayer_1
  properties {soundWaveAssetPtr=&/StickyGrenadeLauncher_DropLand_02}
 }
 {
  id=WavePlayer_2
  properties {soundWaveAssetPtr=&/StickyGrenadeLauncher_DropLand_03}
 }
 {id=WavePlayer_7, properties {soundWaveAssetPtr=&/ImpulseGrenade_DropLayer_01}}
 {id=WavePlayer_8, properties {soundWaveAssetPtr=&/ImpulseGrenade_DropLayer_02}}
]
properties {
 attenuationSettings=&Fort_Weapon_Drop_Att
 maxDistance=3000
 soundClassObject=&Fort_Loot_Collect
 volumeMultiplier=0.8
 attenuationOverrides {falloffDistance=2400, stereoSpread=2000, attenuationShapeExtents {x=800}}
}
