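# Sound cue for the shockwave grenade's warning alarm, routed to the gadget items sound class with a medium explosive-beep attenuation.
# A mixer layers a grenade beep, at slightly reduced volume, with one of two randomly chosen, equally weighted shockwave launcher alarm waves.
# Both layers pass through modulators that apply the same fixed upward pitch shift, and the whole cue plays at reduced volume.
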SoundCue /Game/Sounds/Fort_GamePlay_Sounds/Explosives/ShockwaveGrenade/SC_ShockwaveGrenade_WarningAlarm
root=Mixer_0
nodes [
 {id=Mixer_0, children [Modulator_0,Modulator_1], properties {inputVolume [0.7,1]}}
 {id=Modulator_0, children [WavePlayer_1], properties {pitchMax=1.2, pitchMin=1.2}}
 {id=Modulator_1, children [Random_0], properties {pitchMax=1.2, pitchMin=1.2}}
 {id=Random_0, children [WavePlayer_0,WavePlayer_2], properties {weights [1,1]}}
 {id=WavePlayer_0, properties {soundWaveAssetPtr=/Game/Sounds/Fort_Weapons/Launchers/Shockwave/Shockwave_Launcher_WarningAlarm_01}}
 {id=WavePlayer_1, properties {soundWaveAssetPtr=&/Grenade_Shockwave_Beep01}}
 {id=WavePlayer_2, properties {soundWaveAssetPtr=/Game/Sounds/Fort_Weapons/Launchers/Shockwave/Shockwave_Launcher_WarningAlarm_02}}
]
properties {
 attenuationSettings=/Game/Sounds/Attenuation/Fort_ExplosiveBeep_Medium_Att
 maxDistance=6500
 soundClassObject=&Fort_Gadgets_Items
 volumeMultiplier=0.5
}
